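# Additive unlit material that fakes a flickering or color-shifting light for the Skirmish backpack, tinted by a mip-6 sample of its flipbook texture.
# A camera-aligned mesh, pushed toward the camera, combines a radial glow with fake surface lighting from scene world normals and a pivot-centered sphere mask.
# Depth and synced distance fades attenuate the result.
# Forward and mobile paths output zero emissive and collapse the mesh to its object position.
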Material /BRCosmetics/Accessories/FORT_Backpacks/Backpack_F_MED_Skirmish/Materials/M_Skirmish_FakeLight
graph {
 properties {bEnableResponsiveAA=true, blendMode=BLEND_Additive, pixelDepthOffsetMode=PDOM_Legacy, shadingModel=MSM_Unlit, usedShadingModels=Unlit}
 nodes [
  n1 = ScalarParameter() {defaultValue=40, parameterName=FadeStartDistance}
  n2 = ObjectRadius()
  n3 = ScalarParameter() {defaultValue=10, parameterName=FadeEndDistance}
  n4 = ObjectRadius()
  n5 = MaterialFunctionCall(B=MaterialFunctionCall(Position=ObjectPositionWS(), Fade_Start_Distance=Add(A=@n1, B=@n2), Fade_End_Distance=Add(A=@n3, B=@n4)) {materialFunction=&MF_Distance_SyncedFade_FX}.Result) {materialFunction=&AppendMany}
  n6 = Distance(
   A=MaterialFunctionCall(Exclude WPO Offsets=StaticBool()) {materialFunction=&WorldPositionBehindTranslucency}.Result
   B=WorldPosition().XYZ
  )
  n7 = Multiply(A=ObjectRadius()) {constB=0.8}
  n8 = Constant() {r=1}
  n9 = Constant()
  n10 = Constant() {r=0.5}
  n11 = SphereMask(A=TextureCoordinate(), B=@n10, Radius=@n10, Hardness=Constant())
  n12 = Multiply(A=@n11, B=@n11)
  n13 = VectorParameter() {parameterName=GlowColor, defaultValue=<rgba 1 1 1 0>}
  n14 = MaterialFunctionCall() {materialFunction=&ObjectPivotPoint}
  n15 = MaterialFunctionCall() {materialFunction=&WorldPositionBehindTranslucency}
  n16 = DotProduct(
   A=Normalize(VectorInput=Subtract(A=@n14.Object Pivot Location, B=@n15.Result))
   B=ComponentMask(None=SceneTexture() {sceneTextureId=PPI_WorldNormal}.Color) {b=true, g=true, r=true}
  )
  n17 = Multiply(A=ObjectRadius()) {constB=0.6}
  n18 = ScalarParameter() {defaultValue=0.45, parameterName=FakeLightingMaskHardness}
  n19 = Add(
   A=Multiply(A=If(A=@n6, B=@n7, "A > B"=@n8, "A < B"=@n9), B=Multiply(A=@n12, B=@n13.RGB))
   B=Multiply(A=Saturate(None=@n16), B=SphereMask(A=@n14.Object Pivot Location, B=@n15.Result, Radius=@n17, Hardness=@n18))
  )
  n20 = TextureSample!(
   UVs=MaterialFunctionCall() {materialFunction=/Game/Accessories/FORT_Backpacks/Backpack_F_MED_Skirmish/Materials/MF_Skirmish_FlipbookCurve}.UVs
   Tex
   Level
   Apply View MipBias
  ) {constMipValue=6, mipValueMode=TMVM_MipLevel, texture=/Game/Accessories/FORT_Backpacks/Backpack_F_MED_Skirmish/Materials/T_Skirmish_SubUV_01}
  n21 = Multiply(A=@n20.RGB, B=VectorParameter() {parameterName=ColorScale, defaultValue=<rgba 1 1 1 0>}.RGB)
  n22 = ScalarParameter() {defaultValue=20, parameterName=FadeDistance}
  n23 = Multiply(A=ComponentMask(None=VertexInterpolator(VS=@n5.RGB).PS) {b=true}, B=Multiply(A=Multiply(A=@n19, B=@n21), B=DepthFade(FadeDistance=@n22)))
  n24 = Constant()
  MaterialFunctionCall(Emissive=ShadingPathSwitch(Default=@n23, Deferred=@n23, Forward=@n24, Mobile=@n24)) {materialFunction=&MF_FX_EmissiveAdjustments}
  n25 = Subtract(
   A=CameraPositionWS()
   B=MaterialFunctionCall() {materialFunction=&ObjectPivotPoint}.Object Pivot Location
  )
  n26 = Constant3Vector() {desc=Forward Axis, constant=<rgba -1 0 0 0>}
  Saturate(None=DotProduct(A=Normalize(VectorInput=@n25), B=Transform(None=@n26) {transformSourceType=TRANSFORMSOURCE_Local}))
  ScalarParameter() {defaultValue=0.2, parameterName=GlowOpacity, primitiveDataIndex=5}
  n27 = Constant3Vector() {constant=<rgba 0 0 1 0>}
  n28 = Constant3Vector() {constant=<rgba 0 1 0 0>}
  n29 = Constant3Vector() {constant=<rgba 1 0 0 0>}
  n30 = Subtract(A=CameraPositionWS(), B=ObjectPositionWS())
  n31 = Add(
   A=MaterialFunctionCall!(
    WorldPosition=WorldPosition().XYZ
    PivotLocation
    WS Normals
    Custom Object Basis 1=@n27
    Custom Object Basis 2=@n28
    Custom Object Basis 3=@n29
   ) {materialFunction=/Engine/Functions/Engine_MaterialFunctions02/WorldPositionOffset/AlignMeshToTheCamera}.World Position Offset
   B=Multiply(A=Normalize(VectorInput=@n30), B=ObjectRadius())
  )
  n32 = Subtract(A=ObjectPositionWS(), B=WorldPosition().XYZ)
  ShadingPathSwitch(Default=@n31, Deferred=@n31, Forward=@n32, Mobile=@n32)
  Transform(None=@n27) {transformSourceType=TRANSFORMSOURCE_Local}
  Transform(None=@n28) {transformSourceType=TRANSFORMSOURCE_Local}
  Transform(None=@n29) {transformSourceType=TRANSFORMSOURCE_Local}
 ]
}
